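# Unlit Niagara particle material for the rocket mesh in the sticker-rocket drift trail vehicle cosmetic.
# Color blends the diffuse map with a painted version (particle color times mask, plus the Alpha texture), using dynamic parameters for paint blend and desaturation.
# The blend feeds MF_FX_EmissiveAdjustments.
# A separate chain builds a wandering XYZ drift from blended time-driven sine and cosine waves, scaled per axis and uniformly, plus a WPO_OffsetB vector.
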Material /VehicleCosmetics/DriftTrail/DriftTrail_stickerrocket/Materials/M_Drfit_sticker_rocket_mesh
graph {
 properties {
  bAutomaticallySetUsageInEditor=false
  bUsedWithNiagaraMeshParticles=true
  bUsedWithNiagaraRibbons=true
  bUsedWithNiagaraSprites=true
  shadingModel=MSM_Unlit
  usedShadingModels=Unlit
 }
 nodes [
  n1 = ScalarParameter() {defaultValue=1.5, parameterName=SpeedDB}
  n2 = Time()
  n3 = ScalarParameter() {defaultValue=1, parameterName=SpeedAB}
  n4 = Time()
  n5 = LinearInterpolate(A=Constant() {r=1}, B=Sine(None=Multiply(A=@n1, B=@n2)), Alpha=Sine(None=Multiply(A=@n3, B=@n4)))
  n6 = ScalarParameter() {defaultValue=1, parameterName=X_Drift_Mult}
  n7 = ScalarParameter() {defaultValue=1.5, parameterName=SpeedD}
  n8 = Time()
  n9 = ScalarParameter() {defaultValue=1, parameterName=SpeedA}
  n10 = Time()
  n11 = LinearInterpolate(A=Constant() {r=1}, B=Sine(None=Multiply(A=@n7, B=@n8)), Alpha=Sine(None=Multiply(A=@n9, B=@n10)))
  n12 = ScalarParameter() {defaultValue=1, parameterName=Y_Drift_Mult}
  n13 = Sine(None=Multiply(A=ScalarParameter() {defaultValue=1.5, parameterName=SpeedC}, B=Time()))
  n14 = Constant() {r=-1}
  n15 = Time()
  n16 = ScalarParameter() {defaultValue=1, parameterName=TimeOffset}
  n17 = Cosine(None=Multiply(A=ScalarParameter() {defaultValue=1, parameterName=SpeedB}, B=Add(A=@n15, B=@n16)))
  n18 = AppendVector(
   A=AppendVector(A=Multiply(A=@n5, B=@n6), B=Multiply(A=@n11, B=@n12))
   B=Multiply(A=LinearInterpolate(A=@n13, B=@n14, Alpha=@n17), B=ScalarParameter() {defaultValue=1, parameterName=Z_Drift_Mult})
  )
  n19 = ScalarParameter() {defaultValue=5, parameterName=Uniform_Drift_Mult}
  Add(
   A=ComponentMask(None=Multiply(A=@n18, B=@n19)) {b=true, g=true, r=true}
   B=ComponentMask(None=VectorParameter() {parameterName=WPO_OffsetB}.RGB) {b=true, g=true, r=true}
  )
  MaterialExpressionMeshEmitterVertexColor_4 = ParticleColor()
  n20 = TextureSample() {texture=/VehicleCosmetics/DriftTrail/DriftTrail_stickerrocket/Textures/T_VCVFX_Drift_stickerrocket_Map_Painted_Mask}
  n21 = Multiply(
   A=Add(
    A=Multiply(A=@MaterialExpressionMeshEmitterVertexColor_4.RGB, B=@n20.RGB)
    B=TextureSample() {texture=/VehicleCosmetics/DriftTrail/DriftTrail_stickerrocket/Textures/T_VCVFX_Drift_stickerrocket_Map_Painted_Alpha}.RGB
   )
   B=Constant() {r=1}
  )
  MaterialExpressionMeshEmitterDynamicParameter_9 = DynamicParameter() {paramNames [Painted,Saturation,Black,Param4]}
  MaterialExpressionMeshEmitterDynamicParameter_8 = DynamicParameter() {paramNames [Painted,Saturation,Black,Param4]}
  MaterialFunctionCall(Emissive=LinearInterpolate(
    A=TextureSample() {texture=/VehicleCosmetics/DriftTrail/DriftTrail_stickerrocket/Textures/T_VCVFX_Drift_stickerrocket_SS_Sticker_Rocket_Map_Diffuse}.RGB
    B=Desaturation(None=@n21, Fraction=@MaterialExpressionMeshEmitterDynamicParameter_9.Saturation)
    Alpha=@MaterialExpressionMeshEmitterDynamicParameter_8.Painted
   )) {materialFunction=&MF_FX_EmissiveAdjustments}
 ]
}
