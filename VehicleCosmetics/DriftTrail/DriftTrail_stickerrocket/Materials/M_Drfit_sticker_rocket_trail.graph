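# Translucent unlit Niagara material for the sticker rocket drift trail, mainly sampling a ribbon sticker texture tinted by particle color.
# Emissive combines a U-gradient tint with a LightMult-scaled glow near U zero, passed through an emissive adjustment function.
# Opacity multiplies texture and particle alpha, depth fade, a sphere-mask falloff along U and TrailVisibility.
# World position offset layers time-driven sine drift, a constant offset and borrowed pseudo-3D noise, scaled by dynamic parameters.
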Material /VehicleCosmetics/DriftTrail/DriftTrail_stickerrocket/Materials/M_Drfit_sticker_rocket_trail
graph {
 properties {
  bAutomaticallySetUsageInEditor=false
  bUsedWithMeshParticles=true
  bUsedWithNiagaraMeshParticles=true
  bUsedWithNiagaraRibbons=true
  bUsedWithNiagaraSprites=true
  blendMode=BLEND_Translucent
  shadingModel=MSM_Unlit
  usedShadingModels=Unlit
 }
 nodes [
  n1 = SphereMask(A=TextureCoordinate() {vTiling=0}, B=Constant2Vector()) {attenuationRadius=0.99, hardnessPercent=0}
  MaterialExpressionVertexColor_2 = ParticleColor()
  n2 = TextureSample() {texture=/VehicleCosmetics/DriftTrail/DriftTrail_stickerrocket/Textures/T_VCVFX_Drift_stickerrocket_Sticker_Rocke_Ribbon_Alpha}
  n3 = Multiply(
   A=Multiply(A=@MaterialExpressionVertexColor_2.A, B=@n2.A)
   B=MaterialFunctionCall(
    Opacity In=TextureSample() {texture=/VehicleCosmetics/DriftTrail/DriftTrail_stickerrocket/Textures/T_VCVFX_Drift_stickerrocket_White_Alpha}.A
    DepthFade Distance=Constant() {r=20}
   ) {materialFunction=&MF_VCVFX_DepthFade}.Result
  )
  n4 = DynamicParameter() {paramNames [TrailVisibility, WPO, Noise, LightMult]}
  MaterialExpressionConstantClamp_0 = Saturate(None=Multiply(A=@n1, B=Multiply(A=@n3, B=@n4.TrailVisibility)))
  MaterialExpressionVertexColor_3 = ParticleColor()
  n5 = Multiply(A=@MaterialExpressionVertexColor_3.RGB, B=ComponentMask(None=TextureCoordinate()) {r=true})
  n6 = TextureCoordinate()
  n7 = Power(Base=OneMinus(None=ComponentMask(None=@n6) {r=true}), Exp=Constant() {r=7})
  n8 = DynamicParameter() {paramNames [TrailVisibility, WPO, Noise, LightMult]}
  n9 = Saturate(None=Multiply(A=Multiply(A=@n7, B=@n8.LightMult), B=Multiply(A=@MaterialExpressionVertexColor_2.RGB, B=@n2.RGB)))
  MaterialFunctionCall(Emissive=Add(A=Multiply(A=@n5, B=@n2.RGB), B=Multiply(A=@MaterialExpressionVertexColor_3.RGB, B=@n9))) {materialFunction=&MF_FX_EmissiveAdjustments}
  n10 = Constant() {r=1}
  n11 = Sine(None=Multiply(A=ScalarParameter() {defaultValue=1.5, parameterName=SpeedDB}, B=Time()))
  n12 = Sine(None=Multiply(A=ScalarParameter() {defaultValue=1, parameterName=SpeedAB}, B=Time()))
  n13 = Constant() {r=1}
  n14 = Sine(None=Multiply(A=ScalarParameter() {defaultValue=1.5, parameterName=SpeedD}, B=Time()))
  n15 = Sine(None=Multiply(A=ScalarParameter() {defaultValue=1, parameterName=SpeedA}, B=Time()))
  n16 = AppendVector(
   A=Multiply(A=LinearInterpolate(A=@n10, B=@n11, Alpha=@n12), B=ScalarParameter() {defaultValue=1, parameterName=X_Drift_Mult})
   B=Multiply(A=LinearInterpolate(A=@n13, B=@n14, Alpha=@n15), B=ScalarParameter() {defaultValue=1, parameterName=Y_Drift_Mult})
  )
  n17 = Multiply(A=ScalarParameter() {defaultValue=1.5, parameterName=SpeedC}, B=Time())
  n18 = Multiply(A=ScalarParameter() {defaultValue=1, parameterName=SpeedB}, B=Add(A=Time(), B=ScalarParameter() {defaultValue=1, parameterName=TimeOffset}))
  n19 = Multiply(
   A=LinearInterpolate(A=Sine(None=@n17), B=Constant() {r=-1}, Alpha=Cosine(None=@n18))
   B=ScalarParameter() {defaultValue=1, parameterName=Z_Drift_Mult}
  )
  n20 = ComponentMask(None=Multiply(A=AppendVector(A=@n16, B=@n19), B=ScalarParameter() {defaultValue=5, parameterName=Uniform_Drift_Mult})) {b=true, g=true, r=true}
  n21 = ComponentMask(None=VectorParameter() {parameterName=WPO_OffsetB}.RGB) {b=true, g=true, r=true}
  n22 = MaterialFunctionCall!(
   UV=TextureCoordinate()
   Scale=ScalarParameter() {defaultValue=1, group=Noise, parameterName=NoiseScale}
   aspectRatio=ComponentMask(None=VectorParameter() {group=Noise, parameterName=NoiseAspectRation}.RGB) {g=true, r=true}
   Time=Multiply(A=ScalarParameter() {group=Noise, parameterName=NoiseTime}, B=Time())
  ) {materialFunction=/VehicleCosmetics/Boosts/Boost_Fallingstar/Materials/MF_Boost_Noise_Psuedo3D_01}
  Multiply(A=@n4.WPO, B=Add(A=Add(A=@n20, B=@n21), B=Multiply(A=@n4.Noise, B=@n22.Float Noise)))
  Power(Base=@n1)
 ]
}
