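# Translucent unlit Niagara material for the engine glow of the sticker rocket drift trail, usable on mesh, sprite and ribbon particles.
# Particle color times a dynamic Emissive parameter is masked by inverted Fresnel, keeping view-facing areas brightest, then passed through MF_FX_EmissiveAdjustments.
# A separate chain builds a wobbling XYZ drift vector from time-driven sine and cosine lerps, scaled per axis and uniformly, plus a WPO_OffsetB vector.
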Material /VehicleCosmetics/DriftTrail/DriftTrail_stickerrocket/Materials/M_Drift_sticker_rocket_Engine_Glow
graph {
 properties {
  bAutomaticallySetUsageInEditor=false
  bUsedWithNiagaraMeshParticles=true
  bUsedWithNiagaraRibbons=true
  bUsedWithNiagaraSprites=true
  blendMode=BLEND_Translucent
  shadingModel=MSM_Unlit
  usedShadingModels=Unlit
 }
 nodes [
  n1 = ScalarParameter() {defaultValue=1.5, parameterName=SpeedDB}
  n2 = Time()
  n3 = ScalarParameter() {defaultValue=1, parameterName=SpeedAB}
  n4 = Time()
  n5 = LinearInterpolate(A=Constant() {r=1}, B=Sine(None=Multiply(A=@n1, B=@n2)), Alpha=Sine(None=Multiply(A=@n3, B=@n4)))
  n6 = ScalarParameter() {defaultValue=1, parameterName=X_Drift_Mult}
  n7 = ScalarParameter() {defaultValue=1.5, parameterName=SpeedD}
  n8 = Time()
  n9 = ScalarParameter() {defaultValue=1, parameterName=SpeedA}
  n10 = Time()
  n11 = LinearInterpolate(A=Constant() {r=1}, B=Sine(None=Multiply(A=@n7, B=@n8)), Alpha=Sine(None=Multiply(A=@n9, B=@n10)))
  n12 = ScalarParameter() {defaultValue=1, parameterName=Y_Drift_Mult}
  n13 = Sine(None=Multiply(A=ScalarParameter() {defaultValue=1.5, parameterName=SpeedC}, B=Time()))
  n14 = Constant() {r=-1}
  n15 = Time()
  n16 = ScalarParameter() {defaultValue=1, parameterName=TimeOffset}
  n17 = Cosine(None=Multiply(A=ScalarParameter() {defaultValue=1, parameterName=SpeedB}, B=Add(A=@n15, B=@n16)))
  n18 = AppendVector(
   A=AppendVector(A=Multiply(A=@n5, B=@n6), B=Multiply(A=@n11, B=@n12))
   B=Multiply(A=LinearInterpolate(A=@n13, B=@n14, Alpha=@n17), B=ScalarParameter() {defaultValue=1, parameterName=Z_Drift_Mult})
  )
  n19 = ScalarParameter() {defaultValue=5, parameterName=Uniform_Drift_Mult}
  Add(
   A=ComponentMask(None=Multiply(A=@n18, B=@n19)) {b=true, g=true, r=true}
   B=ComponentMask(None=VectorParameter() {parameterName=WPO_OffsetB, defaultValue=<rgba 0 0 0 1>}.RGB) {b=true, g=true, r=true}
  )
  n20 = Fresnel()
  MaterialExpressionMeshEmitterDynamicParameter_10 = DynamicParameter() {paramNames [Emissive,Param2,Param3,Param4]}
  MaterialExpressionMeshEmitterVertexColor_5 = ParticleColor()
  MaterialFunctionCall(Emissive=Multiply(
    A=OneMinus(None=@n20)
    B=Multiply(A=@MaterialExpressionMeshEmitterDynamicParameter_10.Emissive, B=@MaterialExpressionMeshEmitterVertexColor_5.RGB)
   )) {materialFunction=&MF_FX_EmissiveAdjustments}
 ]
}
